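; BP_AnimAtlas is an art-tool Blueprint that packs the animation textures in its anims array into one atlas render target.
; Its construction script sizes a nearest-filtered RGBA16f target, then pack draws textures left to right, wrapping rows at 2048 pixels.
; Column zero is a lookup table, storing each animation's UV start and extent as one pixel at its index.
; A separate function exports the atlas as an EXR file.
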(event EventBeginPlay)

(event Collision|EventActorBeginOverlap
  (params (OtherActor :type Actor%20Object%20Reference)))

(event EventTick
  (params (DeltaSeconds :type Float%20%28single-precision%29)))

(fn DrawPixel
  (params (renderTarget :type Texture%20Render%20Target%202D%20Object%20Reference) (x :type Integer) (y :type Integer) (color :type Linear%20Color%20Structure))
  (bind (_canvas _size _context) (Rendering|BeginDrawCanvastoRenderTarget renderTarget))
  (Canvas|DrawTexture _canvas "" (Math|Vector2D|MakeVector2D (Math|Conversions|ToFloat(Integer) x) (Math|Conversions|ToFloat(Integer) y)) "(X=1.000000,Y=1.000000)" "(X=0.000000,Y=0.000000)" "(X=1.000000,Y=1.000000)" color "BLEND_Opaque" 0.0 "(X=0.500000,Y=0.500000)")
  (Rendering|EndDrawCanvastoRenderTarget _context))

(fn DrawTexture
  (params (TextureRenderTarget :type Texture%20Render%20Target%202D%20Object%20Reference) (Texture :type Texture%20Object%20Reference) (x :type Integer) (y :type Integer))
  (bind (_canvas _size _context) (Rendering|BeginDrawCanvastoRenderTarget TextureRenderTarget))
  (bind _size_2 (Class|VolumetricsBPLIbrary|GetTextureSize Texture))
  (bind (_size_x _size_y) (Utilities|Struct|BreakIntPoint _size_2))
  (Canvas|DrawTexture _canvas Texture (Math|Vector2D|MakeVector2D (Math|Conversions|ToFloat(Integer) x) (Math|Conversions|ToFloat(Integer) y)) (Math|Vector2D|MakeVector2D (Math|Conversions|ToFloat(Integer) _size_x) (Math|Conversions|ToFloat(Integer) _size_y)) "(X=0.000000,Y=0.000000)" "(X=1.000000,Y=1.000000)" "(R=1.000000,G=1.000000,B=1.000000,A=1.000000)" "BLEND_Opaque" 0.0 "(X=0.500000,Y=0.500000)")
  (Rendering|EndDrawCanvastoRenderTarget _context)
  (return _size_x _size_y))

(fn ExportRenderTargetToStaticTexture
  (bind _exportpath (Variables|Default|GetExportPath))
  (bind _path (Utilities|Struct|BreakDirectoryPath _exportpath))
  (if (Utilities|Paths|DirectoryExists _path)
    (bind _output_get (LocalVariables|Set_path _path))
    (bind _output_get_2 (LocalVariables|Set_filename (Utilities|String|Append "animationSequence" ".exr")))
    (Development|PrintString (Utilities|String|Append _output_get "/" _output_get_2))
    (LocalVariables|Setfilenames (Utilities|Array|MakeArray (Utilities|String|Append _output_get "/" _output_get_2)))
    (Rendering|ExportRenderTarget (Variables|Default|GetinternalRenderTarget) (LocalVariables|Get_path) (LocalVariables|Get_filename))
    (else
      (Development|PrintString "Invalid Path"))))

(fn pack
  (bind _texelwidth (LocalVariables|GettexelWidth))
  (bind _x (LocalVariables|Getx))
  (bind _returnvalue (* _x _texelwidth))
  (bind _texelheight (LocalVariables|GettexelHeight))
  (bind _y (LocalVariables|Gety))
  (bind _returnvalue_2 (* _y _texelheight))
  (bind _halftexelwidth (LocalVariables|GethalfTexelWidth))
  (bind _texelwidth_2 (LocalVariables|GettexelWidth))
  (bind _returnvalue_3 (+ _returnvalue _texelwidth_2))
  (bind _returnvalue_4 (+ _returnvalue_3 _halftexelwidth))
  (bind _texelwidth_3 (LocalVariables|GettexelWidth))
  (bind _texelheight_2 (LocalVariables|GettexelHeight))
  (Rendering|ClearRenderTarget2D (Variables|Default|GetinternalRenderTarget))
  (bind _size (Class|VolumetricsBPLIbrary|GetTextureSize (Variables|Default|GetinternalRenderTarget)))
  (bind (_size_x _size_y) (Utilities|Struct|BreakIntPoint _size))
  (bind _output_get (LocalVariables|Setrt_width _size_x))
  (LocalVariables|Setrt_height _size_y)
  (bind _output_get_2 (LocalVariables|SettexelWidth (/ 1.0 (Math|Conversions|ToFloat(Integer) _output_get))))
  (Development|PrintString (Utilities|String|ToString(Float) (/ 1.0 (Math|Conversions|ToFloat(Integer) _output_get))))
  (LocalVariables|SethalfTexelWidth (* _output_get_2 0.5))
  (LocalVariables|SettexelHeight (/ 1.0 (Math|Conversions|ToFloat(Integer) (LocalVariables|Getrt_height))))
  (LocalVariables|Sety 0)
  (LocalVariables|Setx 1)
  (for (_array_element _array_index) (Variables|Default|Getanims)
    (Utilities|IsValid _array_element
      (:Is%20Valid
        ;[1 1 1] if we'd go over 2048 put the texture into the next row
        (bind _size_2 (Class|VolumetricsBPLIbrary|GetTextureSize _array_element))
        (bind (_size_x_2 _size_y_2) (Utilities|Struct|BreakIntPoint _size_2))
        (if (> (+ _size_x_2 (LocalVariables|Getx)) 2048)
          (LocalVariables|Setx 1)
          (LocalVariables|Sety (+ _size_y_2 (LocalVariables|Gety))))
        (bind _output_get_3 (LocalVariables|Settexture _array_element))
        (bind (_size_x_3 _size_y_3) (CallFunction|DrawTexture (Variables|Default|GetinternalRenderTarget) (LocalVariables|Gettexture) _x _y))
        (CallFunction|DrawPixel 0 (Variables|Default|GetinternalRenderTarget) 0 _array_index (Math|Color|MakeColor _returnvalue_4 _returnvalue_2 (- (* (- _size_x_3 1) _texelwidth_3) _texelwidth_3) (* _size_y_3 _texelheight_2)))
        (Development|PrintString (Utilities|String|Append (Utilities|GetDisplayName _output_get_3) " uStart:" (Utilities|String|ToString(Float) _returnvalue_4) " uWidth: " (Utilities|String|ToString(Float) (- (* (- _size_x_3 1) _texelwidth_3) _texelwidth_3)) " vStart:" (Utilities|String|ToString(Float) _returnvalue_2) " vWidth:" (Utilities|String|ToString(Float) (* _size_y_3 _texelheight_2))))
        (LocalVariables|Setx (+ _size_x_3 (LocalVariables|Getx)))))))

(fn ConstructionScript
  (bind _anims (Variables|Default|Getanims))
  (bind _rowwidth (LocalVariables|GetrowWidth))
  (bind _internalrendertarget (Variables|Default|GetinternalRenderTarget))
  (if (Utilities|Array|IsValidIndex (Variables|Default|Getanims))
    (LocalVariables|SetrowWidth 0)
    (LocalVariables|Setwidth 0)
    (LocalVariables|Setheight 0)
    (LocalVariables|Setx 0)
    (bind _size (Class|VolumetricsBPLIbrary|GetTextureSize (Utilities|Array|Get(acopy) _anims)))
    (bind (_size_x _size_y) (Utilities|Struct|BreakIntPoint _size))
    (LocalVariables|Setheight _size_y)
    (for _array_element _anims
      (Utilities|IsValid _array_element
        (:Is%20Valid
          (bind _size_2 (Class|VolumetricsBPLIbrary|GetTextureSize _array_element))
          (bind (_size_x_2 _size_y_2) (Utilities|Struct|BreakIntPoint _size_2))
          (if (> (+ _rowwidth _size_x_2) 2048)
            (LocalVariables|Setheight (+ (LocalVariables|Getheight) _size_y_2))
            (LocalVariables|SetrowWidth _size_x_2)
            (else
              (bind _output_get (LocalVariables|SetrowWidth (+ _rowwidth _size_x_2)))
              (LocalVariables|Setwidth (Math|Integer|Max(Integer) (LocalVariables|Getwidth) _output_get))))
          (Development|PrintString (Utilities|String|Append (Utilities|GetDisplayName _array_element) " " (Utilities|String|ToString(Integer) _size_x_2) " " (Utilities|String|ToString(Integer) _size_y_2))))))
    (Rendering|CreateAndAssignRenderTarget _internalrendertarget (+ (LocalVariables|Getwidth) 1) (LocalVariables|Getheight) "RTF_RGBA16f"
      (:Out
        (Variables|/Script/Engine.Texture|SetFilter "TF_Nearest" _internalrendertarget)
        (Variables|Default|SetWidth_RT (LocalVariables|Getwidth))
        (Variables|Default|SetHeight_RT (LocalVariables|Getheight))
        (CallFunction|Pack)))))
