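; Editor-side generated mesh actor that bakes a posed character into a skeletal mesh asset, as an art tool.
; Construction applies SkelMesh to the component and sets it to single-node animation with AnimationPose at an initial position.
; On mesh rebuild, if Enabled, Geometry Script copies the component's mesh into the dynamic mesh.
; On success, that result is written into the TarketSkelMesh asset, overwriting its geometry.
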(event EventBeginPlay)

(event Events|EventOnRebuildGeneratedMesh
  (params (TargetMesh :type Dynamic%20Mesh%20Object%20Reference))
  (if (Variables|Default|GetEnabled)
    (bind (_localtoworld _returnvalue) (GeometryScript|Scene|CopyMeshfromComponent (Variables|Default|GetSkeletalMesh) TargetMesh)
      (:Success
        (bind _tarketskelmesh (Variables|Default|GetTarketSkelMesh)
          (:then
            (GeometryScript|SkeletalMesh|CopyMeshtoSkeletalMesh _returnvalue _tarketskelmesh))
          (:else))))))

(event Collision|EventActorBeginOverlap
  (params (OtherActor :type Actor%20Object%20Reference)))

(event EventTick
  (params (DeltaSeconds :type Float%20%28single-precision%29)))

(fn ConstructionScript
  (bind _skeletalmesh (Variables|Default|GetSkeletalMesh))
  (Utilities|FlowControl|Sequence
    (:then_0
      (bind _skelmesh (Variables|Default|GetSkelMesh)
        (:then
          (Components|SkeletalMesh|SetSkeletalMeshAsset (Variables|Default|GetSkeletalMesh) _skelmesh))
        (:else)))
    (:then_1
      (Components|Animation|SetAnimationMode _skeletalmesh "AnimationSingleNode" true)
      (Variables|/Script/Engine.SkeletalMeshComponent|SetAnimationData (Utilities|Struct|MakeSingleAnimationPlayData (Variables|Default|GetAnimationPose) true true (Variables|Default|GetInitial%20Position)) _skeletalmesh))))
