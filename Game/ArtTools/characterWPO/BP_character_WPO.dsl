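; Editor tool actor that bakes a skeletal animation, with an optional prop, into a bone texture for material-driven (WPO) static-mesh crowd characters.
; Construction script builds the bone list from LOD presets, then poses skeletal components at 30 fps.
; Bind pose fills column zero; per-frame location and quaternion deltas, plus optional scale, fill stacked bone rows.
; Dynamic materials preview the render target, which helper functions export as EXR or a texture asset.
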(event EventBeginPlay)

(fn crowdDemo
  (bind _returnvalue (AddComponent|NODE_AddHierarchicalInstancedStaticMeshComponent-0 :RelativeTransform_Rotation "0, 0, 0" :RelativeTransform_Scale "1.000000,1.000000,1.000000"))
  (Components|StaticMesh|SetStaticMesh _returnvalue (Variables|Default|GetstaticMesh))
  (for _ (range (+ (Variables|Crowd|GetcrowdNum) 1))
    (Components|InstancedStaticMesh|AddInstance _returnvalue (Math|Transform|MakeTransform (Math|Vector|MakeVector (Math|Random|RandomFloatinRange -1e+03 1e+03) (Math|Random|RandomFloatinRange -1e+03 1e+03))) false)))

(fn DrawBindPose
  (params (renderTarget :type Texture%20Render%20Target%202D%20Object%20Reference) (bones :type Array%20of%20Strings) (bindPostSkeletalMeshComponent :type Skeletal%20Mesh%20Component%20Object%20Reference) (PROP_bindPoseSkeletalMesh :type Skeletal%20Mesh%20Component%20Object%20Reference))
  (bind _prop_bones (Variables|Prop|GetProp_Bones))
  (bind _returnvalue (Utilities|Array|Length _prop_bones))
  (bind _returnvalue_2 (Utilities|Array|Length bones))
  (bind _returnvalue_3 (- _returnvalue_2 _returnvalue))
  (LocalVariables|SetlocalRenderTarget renderTarget)
  ;[1 1 1] loop through all bones
  (for (_array_element _array_index) bones
    (LocalVariables|Setindex _array_index)
    (Utilities|FlowControl|Sequence
      (:then_0
        (bind (_returnvalue_location _returnvalue_rotation) (Math|Transform|BreakTransform (Transformation|GetSocketTransform (select (< _array_index _returnvalue_3) bindPostSkeletalMeshComponent PROP_bindPoseSkeletalMesh) (Utilities|String|StringToName _array_element) "RTS_Actor")))
        (bind (_returnvalue_rotation_roll _returnvalue_rotation_pitch _returnvalue_rotation_yaw) (Math|Rotator|BreakRotator _returnvalue_rotation))
        (LocalVariables|Setlocation _returnvalue_location)
        ;[1 1 1] draw position
        (CallFunction|DrawPixel (LocalVariables|GetlocalRenderTarget) 0 (LocalVariables|Getindex) (Math|Conversions|ToLinearColor(Vector) (LocalVariables|Getlocation))))
      (:then_1
        (bind _output_get (LocalVariables|SetlocalQuat (Math|Quat|MakefromEuler(Quat) (Math|Vector|MakeVector _returnvalue_rotation_roll _returnvalue_rotation_pitch _returnvalue_rotation_yaw))))
        (Math|Quat|EnsureShortestArcTo(Quat) _output_get (Math|Quat|MakefromEuler(Quat) (1.8e+02 0.0 0.0)))
        (bind _output_get_2 (LocalVariables|Setquat _output_get))
        (bind (_output_get_x _output_get_y _output_get_z _output_get_w) (Math|Quat|BreakQuat _output_get_2))
        ;[1 1 1] draw rotation
        (CallFunction|DrawPixel (LocalVariables|GetlocalRenderTarget) 0 (+ (LocalVariables|Getindex) (Variables|Bones|GetnumBones)) (Utilities|Struct|MakeLinearColor _output_get_x _output_get_y _output_get_z _output_get_w)))
      (:then_2
        (if (Variables|Default|GetincludeScaleAnimation)
          ;[1 1 1] draw scale
          (CallFunction|DrawPixel (LocalVariables|GetlocalRenderTarget) 0 (Math|Integer|int+int (LocalVariables|Getindex) (Variables|Bones|GetnumBones) (Variables|Bones|GetnumBones)) "(R=1.000000,G=1.000000,B=1.000000,A=1.000000)"))))))

(fn DrawFrame
  (params (renderTarget :type Texture%20Render%20Target%202D%20Object%20Reference) (skeletalMeshComponent :type Skeletal%20Mesh%20Component%20Object%20Reference) (bindPostSkeletalMeshComponent :type Skeletal%20Mesh%20Component%20Object%20Reference) (frame :type Integer) (bones :type Array%20of%20Strings) (PROPSkeletalMeshComponent :type Skeletal%20Mesh%20Component%20Object%20Reference) (PROPbindPostSkeletalMeshComponent :type Skeletal%20Mesh%20Component%20Object%20Reference))
  (bind _prop_bones (Variables|Prop|GetProp_Bones))
  (bind _returnvalue (Utilities|Array|Length _prop_bones))
  (bind _returnvalue_2 (Utilities|Array|Length bones))
  (bind _returnvalue_3 (- _returnvalue_2 _returnvalue))
  (bind _skeletalmeshcomponent (LocalVariables|Get_skeletalMeshComponent))
  (bind _propskeletalmeshcomponent (Variables|GetPROPSkeletalMeshComponent))
  (bind _output_prop (Variables|Prop|GetOutput%20Prop))
  (LocalVariables|Set_renderTarget renderTarget)
  (LocalVariables|Set_skeletalMeshComponent skeletalMeshComponent)
  (LocalVariables|Set_bindPoseSkeletalMeshComponent bindPostSkeletalMeshComponent)
  (bind _output_get (LocalVariables|SetlocalFrame (+ frame 1)))
  ;[1 1 1] loop through all bones
  (for (_array_element _array_index) bones
    (LocalVariables|Set_bone _array_element)
    (LocalVariables|Set_index _array_index)
    (Utilities|FlowControl|Sequence
      (:then_0
        (CallFunction|DrawLocation (LocalVariables|Get_renderTarget) (LocalVariables|GetlocalFrame) (LocalVariables|Get_index) (select (select _output_prop (< _array_index _returnvalue_3) true) _skeletalmeshcomponent _propskeletalmeshcomponent) (select (select _output_prop (< _array_index _returnvalue_3) true) (LocalVariables|Get_bindPoseSkeletalMeshComponent) (Variables|GetPROPbindPostSkeletalMeshComponent)) (LocalVariables|Get_bone)))
      (:then_1
        (CallFunction|DrawRotation (LocalVariables|Get_renderTarget) (LocalVariables|GetlocalFrame) (+ (LocalVariables|Get_index) (Variables|Bones|GetnumBones)) (select (select _output_prop (< _array_index _returnvalue_3) true) (LocalVariables|Get_skeletalMeshComponent) (Variables|GetPROPSkeletalMeshComponent)) (select (select _output_prop (< _array_index _returnvalue_3) true) (LocalVariables|Get_bindPoseSkeletalMeshComponent) (Variables|GetPROPbindPostSkeletalMeshComponent)) (LocalVariables|Get_bone)))
      (:then_2
        (if (Variables|Default|GetincludeScaleAnimation)
          (CallFunction|DrawScale (LocalVariables|Get_renderTarget) (LocalVariables|GetlocalFrame) (Math|Integer|int+int (LocalVariables|Get_index) (Variables|Bones|GetnumBones) (Variables|Bones|GetnumBones)) (select (select _output_prop (< _array_index _returnvalue_3) true) (LocalVariables|Get_skeletalMeshComponent) (Variables|GetPROPSkeletalMeshComponent)) (LocalVariables|Get_bone))))
      (:then_3
        (if (== (LocalVariables|Get_index) 56)
          (bind _asskeletal_mesh_component (Utilities|Casting|CastToSkeletalMeshComponent (select (select _output_prop (< _array_index _returnvalue_3) true) _skeletalmeshcomponent _propskeletalmeshcomponent)))
          (Development|PrintString (Utilities|String|Append (Utilities|String|Append (Utilities|String|Append (Utilities|String|Append "DBG_DF = " (Utilities|GetDisplayName (Variables|/Script/Engine.SkinnedMeshComponent|GetSkeletalMesh _asskeletal_mesh_component))) "Value  X = ") (Utilities|String|ToString(Integer) _returnvalue_3)) (Utilities|String|ToString(Boolean) (select _output_prop (< _array_index _returnvalue_3) true)))))))))

(fn DrawLocation
  (params (renderTarget :type Texture%20Render%20Target%202D%20Object%20Reference) (x :type Integer) (y :type Integer) (SkeletalMeshComponent :type Scene%20Component%20Object%20Reference) (BindPoseSkeletalMeshComponent :type Scene%20Component%20Object%20Reference) (bone :type String))
  (bind _returnvalue (Utilities|String|StringToName bone))
  (bind _returnvalue_location (Math|Transform|BreakTransform (Transformation|GetSocketTransform SkeletalMeshComponent _returnvalue "RTS_Actor")))
  (bind _returnvalue_location_2 (Math|Transform|BreakTransform (Transformation|GetSocketTransform BindPoseSkeletalMeshComponent _returnvalue "RTS_Actor")))
  (bind _returnvalue_2 (- _returnvalue_location _returnvalue_location_2))
  ;[1 1 1] draw position
  (CallFunction|DrawPixel renderTarget x y (Math|Conversions|ToLinearColor(Vector) _returnvalue_2))
  (if (Utilities|Name|Equal(Name) _returnvalue "JNT_global_guitar")
    (Development|PrintString (Utilities|String|Append (Utilities|String|Append (Utilities|String|Append "DBG_Draw = " (Utilities|GetDisplayName SkeletalMeshComponent)) "   //// ") (Utilities|String|Append (Utilities|String|ToString(Vector) _returnvalue_location_2) "-" (Utilities|String|ToString(Vector) _returnvalue_location) "=" (Utilities|String|ToString(Vector) _returnvalue_2))))))

(fn DrawPixel
  (params (renderTarget :type Texture%20Render%20Target%202D%20Object%20Reference) (x :type Integer) (y :type Integer) (color :type Linear%20Color%20Structure))
  (bind (_canvas _size _context) (Rendering|BeginDrawCanvastoRenderTarget renderTarget))
  (Canvas|DrawTexture _canvas "" (Math|Vector2D|MakeVector2D (Math|Conversions|ToFloat(Integer) x) (Math|Conversions|ToFloat(Integer) y)) "(X=1.000000,Y=1.000000)" "(X=0.000000,Y=0.000000)" "(X=1.000000,Y=1.000000)" color "BLEND_Opaque" 0.0 "(X=0.500000,Y=0.500000)")
  (Rendering|EndDrawCanvastoRenderTarget _context))

(fn DrawRotation
  (params (renderTarget :type Texture%20Render%20Target%202D%20Object%20Reference) (x :type Integer) (y :type Integer) (SkeletalMeshComponent :type Scene%20Component%20Object%20Reference) (BindPoseSkeletalMeshComponent :type Scene%20Component%20Object%20Reference) (bone :type String))
  (bind (_returnvalue_location _returnvalue_rotation) (Math|Transform|BreakTransform (Transformation|GetSocketTransform SkeletalMeshComponent (Utilities|String|StringToName bone) "RTS_World")))
  (bind (_returnvalue_location_2 _returnvalue_rotation_2) (Math|Transform|BreakTransform (Transformation|GetSocketTransform BindPoseSkeletalMeshComponent (Utilities|String|StringToName bone) "RTS_World")))
  (bind (_returnvalue_roll _returnvalue_pitch _returnvalue_yaw) (Math|Rotator|BreakRotator (Math|Rotator|CombineRotators (Math|Rotator|InvertRotator _returnvalue_rotation_2) _returnvalue_rotation)))
  (bind (_shortestarc_roll _shortestarc_pitch _shortestarc_yaw) (Math|Rotator|BreakRotator (Variables|GetshortestArc_Roll)))
  (bind _output_get (LocalVariables|SetlocalQuat (Math|Quat|MakefromEuler(Quat) (Math|Vector|MakeVector _returnvalue_roll _returnvalue_pitch _returnvalue_yaw))))
  (Math|Quat|EnsureShortestArcTo(Quat) _output_get (Math|Quat|MakefromEuler(Quat) (Math|Vector|MakeVector _shortestarc_roll _shortestarc_pitch _shortestarc_yaw)))
  (bind _output_get_2 (LocalVariables|Setquat _output_get))
  (bind (_output_get_x _output_get_y _output_get_z _output_get_w) (Math|Quat|BreakQuat _output_get_2))
  ;[1 1 1] draw rotation
  (CallFunction|DrawPixel renderTarget x y (Utilities|Struct|MakeLinearColor _output_get_x _output_get_y _output_get_z _output_get_w)))

(fn DrawScale
  (params (renderTarget :type Texture%20Render%20Target%202D%20Object%20Reference) (x :type Integer) (y :type Integer) (SkeletalMeshComponent :type Scene%20Component%20Object%20Reference) (bone :type String))
  (bind (_returnvalue_location _returnvalue_rotation _returnvalue_scale) (Math|Transform|BreakTransform (Transformation|GetSocketTransform SkeletalMeshComponent (Utilities|String|StringToName bone) "RTS_Actor")))
  ;[1 1 1] draw position
  (CallFunction|DrawPixel renderTarget x y (Math|Conversions|ToLinearColor(Vector) _returnvalue_scale)))

(fn ExportRenderTargetToStaticTexture
  (bind _exportpath (Variables|Default|GetExportPath))
  (bind _path (Utilities|Struct|BreakDirectoryPath _exportpath))
  (if (Utilities|Paths|DirectoryExists _path)
    (bind _output_get (LocalVariables|Set_path _path))
    (bind _output_get_2 (LocalVariables|Set_filename (Utilities|String|Append (Utilities|GetDisplayName (Variables|Default|GetskeletalMeshActor)) "_" (Utilities|GetDisplayName (Variables|Default|GetanimationSequence)) ".exr")))
    (Development|PrintString (Utilities|String|Append _output_get "/" _output_get_2))
    (LocalVariables|Setfilenames (Utilities|Array|MakeArray (Utilities|String|Append _output_get "/" _output_get_2)))
    (Rendering|ExportRenderTarget (Variables|Default|GetinternalRenderTarget) (LocalVariables|Get_path) (LocalVariables|Get_filename))
    (else
      (Development|PrintString "Invalid Path"))))

(fn FindClosestBoneLine
  (params (position :type Vector) (bones_start :type Array%20of%20Vectors) (bones_end :type Array%20of%20Vectors) (bones_names :type Array%20of%20Strings))
  (bind _vertexposition (LocalVariables|Get_vertexPosition))
  (LocalVariables|Set_vertexPosition position)
  (LocalVariables|Set_distance 1e+05)
  (LocalVariables|Set_boneName "")
  (LocalVariables|Set_index 0)
  (LocalVariables|Set_point "0, 0, 0")
  (for (_array_element _array_index) bones_start
    (if (< (Math|Vector|VectorLength (- _vertexposition (Math|Vector|FindClosestPointonSegment _vertexposition _array_element (Utilities|Array|Get(acopy) bones_end _array_index)))) (LocalVariables|Get_distance))
      (LocalVariables|Set_distance (Math|Vector|VectorLength (- _vertexposition (Math|Vector|FindClosestPointonSegment _vertexposition _array_element (Utilities|Array|Get(acopy) bones_end _array_index)))))
      (LocalVariables|Set_boneName (Utilities|Array|Get(acopy) bones_names _array_index))
      (LocalVariables|Set_index _array_index)
      (LocalVariables|Set_point (Math|Vector|FindClosestPointonSegment _vertexposition _array_element (Utilities|Array|Get(acopy) bones_end _array_index)))))
  ;[1 1 1] line from vertex to nearest point on bone
  (Rendering|Debug|DrawDebugLine (LocalVariables|Get_vertexPosition) (LocalVariables|Get_point) "(R=1.000000,G=0.005135,B=0.000000,A=1.000000)" 1e+01 0.2)
  (return (LocalVariables|Get_boneName) (LocalVariables|Get_index) (LocalVariables|Get_point)))

(fn getJoints
  (bind _self self)
  (bind _returnvalue (Utilities|GetObjectPathString _self))
  (bind _returnvalue_2 (Utilities|String|Append "py Techart\\WPOBaker\\getJoints.py " _returnvalue " Run"))
  (Development|PrintString _returnvalue_2)
  (Development|ExecuteConsoleCommand _returnvalue_2 (Game|GetPlayerController 0)))

(fn printJoints
  (LocalVariables|SetjointList "")
  (for _array_element (Variables|Bones|Getbones)
    (LocalVariables|SetjointList (Utilities|String|Append (LocalVariables|GetjointList) _array_element ",")))
  (Development|PrintString (Utilities|String|LeftChop (LocalVariables|GetjointList) 1)))

(fn proceduralMesh(deprecated))

(fn processAnim
  (bind _localstaticmesh (Variables|Default|GetlocalStaticMesh))
  (bind _prop_local_skeletalmesh (Variables|Default|GetProp_Local_SkeletalMesh))
  (bind _skeletalmesh (Variables|/Script/Engine.SkinnedMeshComponent|GetSkeletalMesh _prop_local_skeletalmesh))
  (bind _spawnedskeletalmesh (LocalVariables|GetSpawnedSkeletalMesh))
  (bind _spawnedpropskeletalmesh (LocalVariables|GetSpawnedPROPSkeletalMesh))
  (Utilities|IsValid (Variables|Default|GetstaticMesh)
    (:Is%20Valid
      ;[1 1 1] Create a RenderTarget2d
      (Rendering|CreateAndAssignRenderTarget (Variables|Default|GetinternalRenderTarget) (Variables|AnimInfo|GetxRes) (Variables|AnimInfo|GetyRes) "RTF_RGBA16f"
        (:Out
          (Variables|/Script/Engine.Texture|SetFilter (Variables|Default|GetFilter) (Variables|Default|GetinternalRenderTarget))
          (Utilities|FlowControl|Sequence
            (:then_0
              (Components|SkeletalMesh|SetUpdateAnimationinEditor (Variables|Default|GetProp_Local_SkeletalMesh) true)
              ;[1 1 1] PROP create a bindPose skeletalMesh Component ... we will work out deltas against this
              (bind _returnvalue (AddComponent|NODE_AddSkeletalMeshComponent-3))
              (bind _output_get (LocalVariables|SetPROP_bindPoseSkeletalMeshComponent_0 _returnvalue))
              (Components|SkinnedMesh|SetSkinnedAssetandUpdate _output_get (Variables|/Script/Engine.SkinnedMeshComponent|GetSkeletalMesh (Variables|Default|GetProp_Local_SkeletalMesh)) true))
            (:then_1
              (Components|SkeletalMesh|SetUpdateAnimationinEditor (Variables|Default|GetlocalSkeletalMesh) true)
              ;[1 1 1] create a bindPose skeletalMesh Component ... we will work out deltas against this
              (bind _returnvalue_2 (AddComponent|NODE_AddSkeletalMeshComponent-2))
              (bind _output_get_2 (LocalVariables|SetbindPoseSkeletalMeshComponent _returnvalue_2))
              (Components|SkinnedMesh|SetSkinnedAssetandUpdate _output_get_2 (Variables|/Script/Engine.SkinnedMeshComponent|GetSkeletalMesh (Variables|Default|GetlocalSkeletalMesh)) true)
              ;[1 1 1] draw bind pose on first column
              (CallFunction|DrawBindPose (Variables|Default|GetinternalRenderTarget) (select (Variables|Prop|GetOutput%20Prop) (Variables|Prop|GetCombinedBoneList) (Variables|Bones|Getbones)) (LocalVariables|GetbindPoseSkeletalMeshComponent) (LocalVariables|GetPROP_bindPoseSkeletalMeshComponent_0))
              ;[1 1 1] loop through all the frames
              (for _index (range (+ (Variables|AnimInfo|GetanimLengthFrames) 1))
                (Variables|AnimInfo|Setframe _index)
                (Utilities|FlowControl|Sequence
                  (:then_0
                    (if (Variables|Prop|GetOutput%20Prop)
                      ;[1 1 1] make a skeletal mesh component per frame.
                      (bind _returnvalue_3 (AddComponent|NODE_AddSkeletalMeshComponent-4))
                      (bind _output_get_3 (LocalVariables|SetSpawnedPROPSkeletalMesh _returnvalue_3))
                      (Development|PrintString (Utilities|String|Append "DBG_DF SetMesh = " (Utilities|GetDisplayName _skeletalmesh)))
                      (Components|SkeletalMesh|SetSkeletalMeshAsset _output_get_3 _skeletalmesh)
                      (Components|SkinnedMesh|SetSkinnedAssetandUpdate _output_get_3 _skeletalmesh true)
                      ;[1 1 1] set the local skeletal mesh to the current frame
                      (Components|SkeletalMesh|SetUpdateAnimationinEditor _spawnedpropskeletalmesh true)
                      (Components|Animation|OverrideAnimationData _spawnedpropskeletalmesh (Variables|Prop|GetProp_AnimationSequence) false true (* (Math|Conversions|ToFloat(Integer) (Variables|AnimInfo|Getframe)) (/ 1.0 3e+01)) 1.0)))
                  (:then_1
                    ;[1 1 1] make a skeletal mesh component per frame.
                    (bind _returnvalue_4 (AddComponent|NODE_AddSkeletalMeshComponent-1))
                    (bind _output_get_4 (LocalVariables|SetSpawnedSkeletalMesh _returnvalue_4))
                    (Components|SkinnedMesh|SetSkinnedAssetandUpdate _output_get_4 (Variables|/Script/Engine.SkinnedMeshComponent|GetSkeletalMesh (Variables|Default|GetlocalSkeletalMesh)) true)
                    ;[1 1 1] set the local skeletal mesh to the current frame
                    (Components|SkeletalMesh|SetUpdateAnimationinEditor _spawnedskeletalmesh true)
                    (Components|Animation|OverrideAnimationData _spawnedskeletalmesh (Variables|Default|GetanimationSequence) false true (* (Math|Conversions|ToFloat(Integer) (Variables|AnimInfo|Getframe)) (/ 1.0 3e+01)) 1.0)
                    ;[1 1 1] draw the current frame to the render target
                    (CallFunction|DrawFrame (Variables|Default|GetinternalRenderTarget) (LocalVariables|GetSpawnedSkeletalMesh) (LocalVariables|GetbindPoseSkeletalMeshComponent) (Variables|AnimInfo|Getframe) (select (Variables|Prop|GetOutput%20Prop) (Variables|Prop|GetCombinedBoneList) (Variables|Bones|Getbones)) (LocalVariables|GetSpawnedPROPSkeletalMesh) (LocalVariables|GetPROP_bindPoseSkeletalMeshComponent_0))
                    (if (Variables|Default|GetCleanUp)
                      ;[1 1 1] clean up the spawned skeletal mesh component
                      (Components|DestroyComponent (LocalVariables|GetSpawnedSkeletalMesh))))))
              (if (Variables|Default|GetCleanUp)
                (Components|DestroyComponent (LocalVariables|GetbindPoseSkeletalMeshComponent))
                (Components|DestroyComponent (LocalVariables|GetPROP_bindPoseSkeletalMeshComponent_0)))
              (Components|StaticMesh|SetStaticMesh (Variables|Default|GetlocalStaticMesh) (Variables|Default|GetstaticMesh))
              ;[1 1 1] setup material to use the renderTarget
              (for (_array_element _array_index) (Rendering|Material|GetMaterials _localstaticmesh)
                (bind _returnvalue_5 (Rendering|Material|CreateDynamicMaterialInstance _localstaticmesh _array_index _array_element "None" :!class /Script/Engine.PrimitiveComponent))
                (Rendering|Material|SetTextureParameterValue _returnvalue_5 "AnimationTexture" (Variables|Default|GetinternalRenderTarget))
                (Rendering|Material|SetScalarParameterValue _returnvalue_5 "NumFrames" (Math|Conversions|ToFloat(Integer) (Variables|AnimInfo|GetxRes)) :!class /Script/Engine.MaterialInstanceDynamic)
                (CallFunction|SyncAnim)))))))))

(fn SaveRenderTargetAndAssingToAsset
  (bind _returnvalue (Game|SpawnActorfromClass "/Game/ArtTools/characterWPO/EUBP_CreateTextureAndAssign.EUBP_CreateTextureAndAssign_C" (Math|Transform|MakeTransform "0, 0, 0") "Undefined" "MultiplyWithRoot" 0 (Variables|Default|GetstaticMesh) (Variables|Default|GetinternalRenderTarget) 0))
  (Actor|DestroyActor _returnvalue))

(fn syncAnim
  (Utilities|FlowControl|Sequence
    (:then_0
      (Components|Animation|SetPosition (Variables|Default|GetProp_Local_SkeletalMesh) 0.0 true))
    (:then_1
      (Components|Animation|SetPosition (Variables|Default|GetlocalSkeletalMesh) 0.0 true)
      (for _array_element (Rendering|Material|GetMaterials (Variables|Default|GetlocalStaticMesh))
        (bind _asmaterial_instance_dynamic (Utilities|Casting|CastToMaterialInstanceDynamic _array_element))
        (Development|PrintString (Utilities|GetDisplayName _asmaterial_instance_dynamic))
        (Rendering|Material|SetScalarParameterValue _asmaterial_instance_dynamic "TimeOffset" (* (Utilities|Time|GetGameTimeinSeconds) -1.0) :!class /Script/Engine.MaterialInstanceDynamic)))))

(fn UpdateAnim
  (bind _localskeletalmesh (Variables|Default|GetlocalSkeletalMesh))
  (Components|SkeletalMesh|SetUpdateAnimationinEditor _localskeletalmesh true)
  (Components|Animation|OverrideAnimationData _localskeletalmesh (Variables|Default|GetanimationSequence) false true (* (Math|Conversions|ToFloat(Integer) (Variables|AnimInfo|Getframe)) (/ 1.0 3e+01)) 0.0))

(fn ConstructionScript
  (bind _output_prop (Variables|Prop|GetOutput%20Prop))
  (bind _combinedbonelist (Variables|Prop|GetCombinedBoneList))
  (bind _bones (Variables|Bones|Getbones))
  (bind _returnvalue (select _output_prop _combinedbonelist _bones))
  (bind _returnvalue_2 (Utilities|Array|Length _returnvalue))
  (bind _animationsequence (Variables|Default|GetanimationSequence))
  (bind _sequencelength (Variables|/Script/Engine.AnimSequenceBase|GetSequenceLength _animationsequence))
  (bind _returnvalue_3 (* 30 _sequencelength))
  (bind _returnvalue_4 (Math|Float|Truncate _returnvalue_3))
  (bind _skeletalmeshactor (Variables|Default|GetskeletalMeshActor))
  (bind _skeletalmeshcomponent (Variables|/Script/Engine.SkeletalMeshActor|GetSkeletalMeshComponent _skeletalmeshactor))
  (bind _localskeletalmesh (Variables|Default|GetlocalSkeletalMesh))
  (bind _skeletalmesh (Variables|/Script/Engine.SkinnedMeshComponent|GetSkeletalMesh _skeletalmeshcomponent))
  (bind _bonequalitypreset (Variables|Default|GetBoneQualityPreset))
  (if (Variables|Default|GetEnabled)
    (Utilities|FlowControl|Sequence
      (:then_0
        ;[1 1 1] Fix for Copy Paste Bone List
        (for (_array_element _array_index) (Variables|Prop|GetProp_Bones)
          (Utilities|Array|SetArrayElem (Variables|Prop|GetProp_Bones) _array_index (Utilities|String|StringToName (Utilities|String|Replace (Utilities|String|ToString(Name) _array_element) "\"")))))
      (:then_1
        (Utilities|IsValid (Variables|Default|GetanimationSequence)
          (:Is%20Valid
            (Utilities|IsValid (Variables|Default|GetskeletalMeshActor)
              (:Is%20Valid
                (Utilities|FlowControl|Sequence
                  (:then_0
                    (Components|StaticMesh|SetStaticMesh (Variables|Default|GetlocalStaticMesh) (Variables|Default|GetstaticMesh))
                    ;[1 1 1] set LOD
                    (Components|SkinnedMesh|SetForcedLOD (Variables|Default|GetlocalSkeletalMesh) (Variables|Bones|Getlod))
                    (Utilities|FlowControl|Sequence
                      (:then_0
                        (if (Variables|Prop|GetOutput%20Prop)
                          (Components|SkinnedMesh|SetSkinnedAssetandUpdate (Variables|Default|GetProp_Local_SkeletalMesh) (Variables|Prop|GetProp_SkeletalMesh) true)
                          ;[1 1 1] set the animation to allow preview
                          (Components|Animation|OverrideAnimationData (Variables|Default|GetProp_Local_SkeletalMesh) (Variables|Prop|GetProp_AnimationSequence) true true (* (Math|Conversions|ToFloat(Integer) (Variables|AnimInfo|Getframe)) (/ 1.0 3e+01)) 1.0)))
                      (:then_1
                        ;[1 1 1] set localSkeletalMesh to one that matches the selected mannequin and has preview geo
                        (Development|PrintString (Utilities|GetDisplayName _skeletalmesh))
                        (if (Utilities|Equal(Object) _skeletalmesh "/Game/Characters/Player/Female/Medium/Base/SK_M_Female_Base_Skeleton.SK_M_Female_Base_Skeleton")
                          (Components|SkinnedMesh|SetSkinnedAssetandUpdate _localskeletalmesh "/Game/Characters/Player/Female/Medium/Base/F_MED_Nude_Base_01.F_MED_Nude_Base_01" true)
                          (elif (Utilities|Equal(Object) _skeletalmesh "/Game/Characters/Player/Male/Medium/Base/SK_M_MALE_Base_Skeleton.SK_M_MALE_Base_Skeleton")
                            (Components|SkinnedMesh|SetSkinnedAssetandUpdate _localskeletalmesh "/Game/Characters/Player/Male/Medium/Base/SK_M_MALE_Base.SK_M_MALE_Base" true)
                            (else
                              (Components|SkinnedMesh|SetSkinnedAssetandUpdate _localskeletalmesh _skeletalmesh true))))
                        ;[1 1 1] set the animation to allow preview
                        (Components|Animation|OverrideAnimationData (Variables|Default|GetlocalSkeletalMesh) (Variables|Default|GetanimationSequence) true true (* (Math|Conversions|ToFloat(Integer) (Variables|AnimInfo|Getframe)) (/ 1.0 3e+01)) 1.0)
                        ;[1 1 1] anim length feedsback the time in seconds
                        (Variables|AnimInfo|SetanimLength _sequencelength)
                        (Variables|AnimInfo|SetanimLengthFrames _returnvalue_4)
                        (Variables|AnimInfo|SetxRes (+ _returnvalue_4 1))
                        (if (Utilities|Array|IsValidIndex (Variables|Bones|GetboneLODs))
                          ;[1 1 1] bone list
                          (Variables|Bones|Setbones (Utilities|String|ParseIntoArray (Utilities|Array|Get(acopy) (Variables|Bones|GetboneLODs) (Variables|Bones|Getlod)) ","))
                          (Utilities|FlowControl|Sequence
                            (:then_0
                              (switch Utilities|FlowControl|Switch|SwitchonCrowd_Baker_Presets _bonequalitypreset
                                (:NewEnumerator0
                                  _)
                                (:NewEnumerator1
                                  _))
                              (Utilities|Array|Clear (Variables|Bones|Getbones))
                              (for _array_element_2 (Utilities|Select :Index _bonequalitypreset :NewEnumerator0 (Variables|/Game/ArtTools/characterWPO/BonePreset.BonePreset_C|GetBones (Variables|Default|GetMediumPreset)) :NewEnumerator1 (Variables|/Game/ArtTools/characterWPO/BonePreset.BonePreset_C|GetBones (Variables|Default|GetHighPreset)))
                                (Utilities|Array|Add (Variables|Bones|Getbones) (Utilities|String|ToString(Name) _array_element_2))))
                            (:then_1
                              (if (Variables|Prop|GetOutput%20Prop)
                                (Utilities|Array|Clear (Variables|Prop|GetCombinedBoneList))
                                (if (Variables|Prop|GetOutput%20Prop_Only)
                                  (for _array_element_3 (Variables|Prop|GetProp_Bones)
                                    (Utilities|Array|Add (Variables|Prop|GetCombinedBoneList) (Utilities|String|ToString(Name) _array_element_3)))
                                  (else
                                    (for _array_element_4 (Variables|Bones|Getbones)
                                      (Utilities|Array|Add (Variables|Prop|GetCombinedBoneList) _array_element_4))
                                    (for _array_element_5 (Variables|Prop|GetProp_Bones)
                                      (Utilities|Array|Add (Variables|Prop|GetCombinedBoneList) (Utilities|String|ToString(Name) _array_element_5)))
                                    (for _array_element_6 (Variables|Prop|GetCombinedBoneList)
                                      (Development|PrintString (Utilities|String|Append "DBG_COMBINED = " _array_element_6)))))))
                            (:then_2
                              (Variables|Bones|SetnumBones _returnvalue_2)
                              (if (Variables|Default|GetincludeScaleAnimation)
                                (Variables|AnimInfo|SetyRes (* _returnvalue_2 3))
                                (else
                                  (Variables|AnimInfo|SetyRes (* _returnvalue_2 2))))))))))
                  (:then_1
                    (Utilities|IsValid (Variables|Default|GetanimationSequence)
                      (:Is%20Valid
                        (Utilities|IsValid (Variables|Default|GetstaticMesh)
                          (:Is%20Valid
                            (Utilities|IsValid (Variables|Default|GetskeletalMeshActor)
                              (:Is%20Valid
                                (CallFunction|ProcessAnim)
                                (CallFunction|SyncAnim)))))))))))))))))
