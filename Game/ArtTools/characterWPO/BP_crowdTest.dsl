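; Test actor that builds a static mesh crowd at construction, apparently for character WPO work in ArtTools.
; The construction script calls crowdTest, which adds one Hierarchical Instanced Static Mesh component per mesh in the staticMesh array.
; Each component receives crowdNum + 1 local-space instances at random XY offsets within ±Spread, without random rotation or scale.
; Placement is unseeded, so it reshuffles on each construction; the event graph stubs are empty.
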(event EventBeginPlay)

(event Collision|EventActorBeginOverlap
  (params (OtherActor :type Actor%20Object%20Reference)))

(event EventTick
  (params (DeltaSeconds :type Float%20%28single-precision%29)))

(fn crowdTest
  (for _array_element (Variables|Default|GetstaticMesh)
    (bind _returnvalue (AddComponent|NODE_AddHierarchicalInstancedStaticMeshComponent-1 :RelativeTransform_Rotation "0, 0, 0" :RelativeTransform_Scale "1.000000,1.000000,1.000000"))
    (Components|StaticMesh|SetStaticMesh _returnvalue _array_element)
    (for _ (range (+ (Variables|Default|GetcrowdNum) 1))
      (Components|InstancedStaticMesh|AddInstance _returnvalue (Math|Transform|MakeTransform (Math|Vector|MakeVector (Math|Random|RandomFloatinRange (* (Variables|Default|GetSpread) -1.0) (Variables|Default|GetSpread)) (Math|Random|RandomFloatinRange (* (Variables|Default|GetSpread) -1.0) (Variables|Default|GetSpread)))) false))))

(fn ConstructionScript
  (CallFunction|CrowdTest))
