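; Editor generated-mesh tool that bakes a Fortnite character's skeletal meshes and optional prop into one static mesh for WPO-animated crowds.
; Construction gathers character parts and preset bone lists; rebuild merges parts, culls by vertex color, then writes animation data into vertex colors, or UVs for Nanite.
; It creates textured material instances and optional simplified LODs.
; An experimental last LOD sweeps limb/spine ribbons, copying nearest-point vertex colors from the baked mesh.
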(event EventBeginPlay)

(event Collision|EventActorBeginOverlap
  (params (OtherActor :type Actor%20Object%20Reference)))

(event EventTick
  (params (DeltaSeconds :type Float%20%28single-precision%29)))

(event Events|EventOnRebuildGeneratedMesh
  (params (TargetMesh :type Dynamic%20Mesh%20Object%20Reference))
  (bind _combinedbonelist (Variables|Prop|GetCombinedBoneList))
  (bind _combinedbonelist_2 (Variables|Prop|GetCombinedBoneList))
  (bind _generateribbonlodforlastlod (Variables|Default|GetGenerateRibbonLODforLastLOD))
  (bind _output_get (Variables|Default|SetTarget%20Mesh TargetMesh))
  (GeometryScript|Materials|EnableMaterialIDs _output_get)
  (if (Variables|Default|GetEnabled)
    (Utilities|FlowControl|Sequence
      (:then_0
        (if (Variables|Prop|GetOutput%20Prop)
          (bind _prop_skeletalmesh (Variables|Prop|GetProp_SkeletalMesh)
            (:then
              (Utilities|Array|Add (Variables|Default|GetskeletalMeshes) _prop_skeletalmesh)
              (if (Variables|Prop|GetOutputPropOnly)
                (Utilities|Array|Clear _combinedbonelist)
                (for _array_element (Variables|Prop|GetProp_Bones)
                  (Utilities|Array|Add _combinedbonelist _array_element))
                (else
                  (Utilities|Array|Clear _combinedbonelist_2)
                  (Utilities|Array|AppendArray _combinedbonelist_2 (Variables|Default|GetboneNames))
                  (for _array_element_2 (Variables|Prop|GetProp_Bones)
                    (Utilities|Array|Add _combinedbonelist_2 _array_element_2)))))
            (:else))))
      (:then_1
        ;[1 1 1] make the mesh
        (CallFunction|MakeStaticMeshSkelMeshes (Variables|Default|GetTarget%20Mesh))
        (Utilities|IsValid (Variables|Default|GettargetStaticMesh)
          (:Is%20Valid
            ;[1 1 1] bake the mesh to a staticMeshAsset
            (Utilities|FlowControl|Sequence
              (:then_0
                ;[1 1 1] Add Material For RibbonLOd
                (if (Variables|Default|GetGenerateRibbonLODforLastLOD)
                  (Utilities|Array|Add (Variables|Default|Getmaterials) "/Game/ArtTools/characterWPO/MI_StaticMeshMaker.MI_StaticMeshMaker")))
              (:then_1
                (GeometryScript|StaticMesh|CopyMeshtoStaticMesh (Variables|Default|GetTarget%20Mesh) (Variables|Default|GettargetStaticMesh) (Utilities|Struct|MakeGeometryScriptCopyMeshToAssetOptions false true false "DoNothing" false false true true "ReplaceAssetMaterials" (Variables|Default|Getmaterials) (Variables|Default|GetmaterialNames) false "MatchTargetLODSetting" true "(bEnabled=True,FallbackPercentTriangles=100.000000,FallbackRelativeError=0.000000)" (Utilities|Struct|MakeMeshNaniteSettings (Variables|Default|GetNaniteEnabled) false false)) (struct :bWriteHiResSource false :LODIndex 0) true
                  (:Success
                    (Utilities|FlowControl|Sequence
                      (:then_1
                        ;[1 1 1] Materials
                        (CallFunction|CreateMaterials)
                        ;[1 1 1] LODs
                        (if (Variables|Default|GetCopyExisitngLODs)
                          (Utilities|FlowControl|Sequence
                            (:then_0
                              (CallFunction|GenerateLOD 1))
                            (:then_1
                              (CallFunction|GenerateLOD 2))
                            (:then_2
                              (CallFunction|GenerateLOD 3))
                            (:then_3
                              (if _generateribbonlodforlastlod
                                (CallFunction|GenerateLOD (Utilities|Array|LastIndex (Variables|Default|GetAutoGenerateLOD_reduction)) _generateribbonlodforlastlod)))
                            (:then_4
                              (DynamicMeshActor|ReleaseAllComputeMeshes))))))))))))))))

(fn clearArrays
  (Utilities|Array|Clear (Variables|Default|Getmaterials))
  (Utilities|Array|Clear (Variables|Default|GetmaterialNames))
  (Utilities|Array|Clear (Variables|Default|GetdiffuseTextures))
  (Utilities|Array|Clear (Variables|Default|GetnormalTextures))
  (Utilities|Array|Clear (Variables|Default|GetskeletalMeshes)))

(fn createMaterial
  (params (index :type Integer))
  (bind _returnvalue (Utilities|String|ToString(Integer) index))
  (bind _targetmeshpath (LocalVariables|GettargetMeshPath))
  (bind _lefts (Utilities|String|Split _targetmeshpath "/" "IgnoreCase" "FromEnd"))
  (bind (_lefts_2 _rights) (Utilities|String|Split _targetmeshpath "." "IgnoreCase" "FromEnd"))
  (bind _returnvalue_2 (Utilities|String|Append "MI_" _rights "_" _returnvalue))
  (bind _targetstaticmesh (Variables|Default|GettargetStaticMesh))
  (Utilities|IsValid _targetstaticmesh
    (:Is%20Valid
      (LocalVariables|SettargetMeshPath (Utilities|GetObjectPathString _targetstaticmesh))
      (Development|PrintString _targetmeshpath false false)
      (LocalVariables|SetassetPath _lefts)
      (LocalVariables|SetmaterialName _returnvalue_2)
      (LocalVariables|SetfullPath (Utilities|String|Append _lefts "/" _returnvalue_2))
      (Development|PrintString (Utilities|String|Append "targetMesh " (LocalVariables|GetassetPath) " " (LocalVariables|GetmaterialName)) false false)
      (bind _returnvalue_3 (EditorScripting|Asset|DoesAssetExist (LocalVariables|GetfullPath) :!class /Script/EditorScriptingUtilities.EditorAssetLibrary))
      (if _returnvalue_3
        (Development|PrintString (Utilities|String|Append "material " (LocalVariables|GetmaterialName) " exists already") false false)
        (bind _returnvalue_4 (EditorScripting|Asset|FindAssetData (LocalVariables|GetfullPath) :!class /Script/EditorScriptingUtilities.EditorAssetLibrary))
        (LocalVariables|SetmaterialObject (AssetData|GetAsset _returnvalue_4))
        (else
          (Development|PrintString (Utilities|String|Append "making material " (LocalVariables|GetmaterialName)) false false)
          (bind _returnvalue_5 (EditorScripting|AssetTools|GetAssetTools))
          (bind _returnvalue_6 (EditorScripting|AssetTools|CreateAsset _returnvalue_5 (LocalVariables|GetmaterialName) (LocalVariables|GetassetPath) "/Script/Engine.MaterialInstanceConstant" 0 "None" false))
          (LocalVariables|SetmaterialObject _returnvalue_6)))
      (bind _asmaterial_instance_constant (Utilities|Casting|CastToMaterialInstanceConstant (LocalVariables|GetmaterialObject)))
      (return _asmaterial_instance_constant))
    (:Is%20Not%20Valid
      (return 0))))

(fn createMaterials
  (bind _diffusetextures (Variables|Default|GetdiffuseTextures))
  (bind _returnvalue (Utilities|Array|LastIndex _diffusetextures))
  (bind _returnvalue_2 (+ _returnvalue 1))
  (Utilities|FlowControl|Sequence
    (:then_0
      (for (_array_element _array_index) _diffusetextures
        (bind _asmaterial_instance_constant (CallFunction|CreateMaterial _array_index))
        (MaterialEditing|SetMaterialInstanceParent _asmaterial_instance_constant (Selection|SelectMaterialInstance "/Game/ArtTools/characterWPO/MI_StaticMeshMaker_Nanite.MI_StaticMeshMaker_Nanite" "/Game/ArtTools/characterWPO/MI_StaticMeshMaker.MI_StaticMeshMaker" (Variables|Default|GetNaniteEnabled)))
        (MaterialEditing|SetMaterialInstanceTextureParameterValue _asmaterial_instance_constant "baseColor" _array_element)
        (MaterialEditing|SetMaterialInstanceTextureParameterValue _asmaterial_instance_constant "normals" (Utilities|Array|Get(acopy) (Variables|Default|GetnormalTextures) _array_index))
        (StaticMesh|SetMaterial (Variables|Default|GettargetStaticMesh) _array_index _asmaterial_instance_constant)))
    (:then_1
      (if (Variables|Default|GetGenerateRibbonLODforLastLOD)
        (bind _asmaterial_instance_constant_2 (CallFunction|CreateMaterial _returnvalue_2))
        (MaterialEditing|SetMaterialInstanceParent _asmaterial_instance_constant_2 "/Game/ArtTools/characterWPO/M_characterWPO_Master_Ribbon.M_characterWPO_Master_Ribbon")
        (StaticMesh|SetMaterial (Variables|Default|GettargetStaticMesh) _returnvalue_2 _asmaterial_instance_constant_2)))))

(fn DeleteMeshByVertexColor
  (params (TargetMesh :type Dynamic%20Mesh%20Object%20Reference) (CurrentSkelMesh :type Object%20Reference))
  (bind _vertexcolorhiding (Variables|VertexColorHiding|GetVertexColorHiding))
  (bind _keys (Utilities|Map|Keys _vertexcolorhiding))
  (bind _values (Utilities|Map|Values _vertexcolorhiding))
  (for (_array_element _array_index) _keys
    (if (== _array_element CurrentSkelMesh)
      (Development|PrintString (Utilities|String|Append "DBGVRTX_ ; " (Utilities|GetDisplayName _array_element) ""))
      (bind _colorlist (GeometryScript|VertexColor|GetMeshPerVertexColors TargetMesh))
      (bind _colorarray (GeometryScript|ListUtils|ConvertColorListtoArray _colorlist))
      (for (_array_element_2 _array_index_2) _colorarray
        (Development|PrintString (Utilities|String|Append (Utilities|String|Append "DBGVRTX_ ; " (Utilities|GetDisplayName _array_element) "") " VERTID = " (Utilities|String|ToString(Integer) _array_index_2) "Color = " (Utilities|String|ToString(LinearColor) _array_element_2) "   Color Search = " (Utilities|String|ToString(LinearColor) (Utilities|Array|Get(aref) _values _array_index))))
        (if (Math|Boolean|ANDBoolean (Math|Float|NearlyEqual(Float) (Utilities|Struct|BreakLinearColor _array_element_2) (Utilities|Struct|BreakLinearColor (Utilities|Array|Get(aref) _values _array_index)) 0.1) (Math|Float|NearlyEqual(Float) (.G (Utilities|Struct|BreakLinearColor _array_element_2)) (.G (Utilities|Struct|BreakLinearColor (Utilities|Array|Get(aref) _values _array_index))) 0.1) (Math|Float|NearlyEqual(Float) (.B (Utilities|Struct|BreakLinearColor (Utilities|Array|Get(aref) _values _array_index))) (.B (Utilities|Struct|BreakLinearColor _array_element_2)) 0.1))
          (Utilities|Array|Add (LocalVariables|GetVertexIDs) _array_index_2)))
      (bind _indexlist (GeometryScript|ListUtils|ConvertArraytoIndexList (LocalVariables|GetVertexIDs) "Vertex"))
      (GeometryScript|MeshEdits|DeleteVerticesfromMesh TargetMesh _indexlist))))

(fn generateLOD
  (params (lod :type Integer) (GenerateRibbonLOD :type Boolean))
  (bind _returnvalue (DynamicMeshActor|AllocateComputeMesh))
  (Variables|Default|SetnumMaterialIds 0)
  (Utilities|Array|Clear (Variables|Default|GetdiffuseTextures))
  (bind _dynamicmesh (CallFunction|MakeStaticMeshSkelMeshes _returnvalue lod GenerateRibbonLOD))
  (GeometryScript|StaticMesh|CopyMeshtoStaticMesh _dynamicmesh (Variables|Default|GettargetStaticMesh) (Utilities|Struct|MakeGeometryScriptCopyMeshtoAssetOptions false true) (Utilities|Struct|MakeGeometryScriptMeshWriteLOD false lod) true))

(fn getParentBone
  (params (bone :type Name))
  (if (Utilities|String|Equal,CaseInsensitive(String) (Utilities|String|ToString(Name) bone) (Utilities|String|ToString(Name) (Variables|Default|GetRootBoneName)))
    (return bone)
    (else
      (Development|PrintString (Utilities|String|Append "Current Bone = " (Utilities|String|ToString(Name) bone)) false)
      (if (Utilities|Array|ContainsItem (select (Variables|Prop|GetOutput%20Prop) (Variables|Prop|GetCombinedBoneList) (Variables|Default|GetboneNames)) bone)
        (return bone)
        (else
          (bind _return (CallFunction|GetParentBone (Components|SkinnedMesh|GetParentBone (Variables|Default|GetlocalSkeletalMesh) bone)))
          (return _return))))))

(fn getValidBone
  (params (index :type Integer))
  (bind _localskeletalmesh (Variables|Default|GetlocalSkeletalMesh))
  (bind _returnvalue (Components|SkinnedMesh|GetBoneName _localskeletalmesh index))
  (bind _return (CallFunction|GetParentBone _returnvalue))
  (return _return _returnvalue (Utilities|Array|FindItem (select (Variables|Prop|GetOutput%20Prop) (Variables|Prop|GetCombinedBoneList) (Variables|Default|GetboneNames)) _return) (/ (Utilities|Array|FindItem (select (Variables|Prop|GetOutput%20Prop) (Variables|Prop|GetCombinedBoneList) (Variables|Default|GetboneNames)) _return) (Utilities|Array|Length (select (Variables|Prop|GetOutput%20Prop) (Variables|Prop|GetCombinedBoneList) (Variables|Default|GetboneNames))))))

(fn isArmBone
  (params (bone :type Name))
  (if (Utilities|Array|ContainsItem (Utilities|Array|MakeArray "clavicle_l" "clavicle_r") bone)
    (return true (== bone "clavicle_l"))
    (elif (== bone "")
      (return false false)
      (else
        (bind (_arm _left) (CallFunction|IsArmBone (Components|SkinnedMesh|GetParentBone (Variables|Default|GetlocalSkeletalMesh) bone)))
        (return _arm _left)))))

(fn isArmVertex
  (params (boneIndex :type Integer))
  (bind (_arm _left) (CallFunction|IsArmBone (Components|SkinnedMesh|GetBoneName (Variables|Default|GetlocalSkeletalMesh) boneIndex)))
  (return _arm _left))

(fn MakeRibbonMesh
  (params (TargetMesh :type Dynamic%20Mesh%20Object%20Reference) (SkelMeshDynamic :type Dynamic%20Mesh%20Object%20Reference))
  (bind _array (Utilities|Array|MakeArray 0.0 1.0))
  (bind _array_2 (Utilities|Array|MakeArray "(X=-1.000000,Y=0.000000)" "(X=1.000000,Y=0.000000)"))
  (bind _skelmeshdynamic (Variables|GetSkelMeshDynamic))
  (bind _skelmeshdynamic_2 (Variables|GetSkelMeshDynamic))
  (bind _skelmeshdynamic_3 (Variables|GetSkelMeshDynamic))
  (bind _skelmeshdynamic_4 (Variables|GetSkelMeshDynamic))
  (bind _skelmeshdynamic_5 (Variables|GetSkelMeshDynamic))
  (Utilities|FlowControl|Sequence
    (:then_0
      (bind (_bisvalidbonename _boneinfo) (GeometryScript|MeshQueries|BoneWeights|GetBoneInfo _skelmeshdynamic_2 "upperarm_l"))
      (bind (_bisvalidbonename_2 _boneinfo_2) (GeometryScript|MeshQueries|BoneWeights|GetBoneInfo _skelmeshdynamic_2 "lowerarm_l"))
      (bind (_bisvalidbonename_3 _boneinfo_3) (GeometryScript|MeshQueries|BoneWeights|GetBoneInfo _skelmeshdynamic_2 "hand_l"))
      (bind _returnvalue (GeometryScript|Primitives|AppendSweepPolyline :TargetMesh (Variables|GetTargetMesh) :PrimitiveOptions (Utilities|Struct|MakeGeometryScriptPrimitiveOptions "SingleGroup" false "ScaleToFill") :PolylineVertices _array_2 :SweepPath (Utilities|Array|MakeArray (.WorldTransform (Utilities|Struct|BreakBoneInfo _boneinfo)) (.WorldTransform (Utilities|Struct|BreakBoneInfo _boneinfo_2)) (.WorldTransform (Utilities|Struct|BreakBoneInfo _boneinfo_3)) (Math|Transform|MakeTransform (+ (.location (.WorldTransform (Utilities|Struct|BreakBoneInfo _boneinfo_3))) "0.000000,0.000000,-15.000000") (.rotation (.WorldTransform (Utilities|Struct|BreakBoneInfo _boneinfo_3))) (.scale (.WorldTransform (Utilities|Struct|BreakBoneInfo _boneinfo_3))))) :PolylineTexParamU _array :bLoop false :StartScale 1.0 :EndScale 1.0 :RotationAngleDeg 0.0 :MiterLimit 1.0)))
    (:then_1
      (bind (_bisvalidbonename_4 _boneinfo_4) (GeometryScript|MeshQueries|BoneWeights|GetBoneInfo _skelmeshdynamic_5 "upperarm_r"))
      (bind (_bisvalidbonename_5 _boneinfo_5) (GeometryScript|MeshQueries|BoneWeights|GetBoneInfo _skelmeshdynamic_5 "lowerarm_r"))
      (bind (_bisvalidbonename_6 _boneinfo_6) (GeometryScript|MeshQueries|BoneWeights|GetBoneInfo _skelmeshdynamic_5 "hand_r"))
      (GeometryScript|Primitives|AppendSweepPolyline :TargetMesh (Variables|GetTargetMesh) :PrimitiveOptions (Utilities|Struct|MakeGeometryScriptPrimitiveOptions "PerFace") :PolylineVertices _array_2 :SweepPath (Utilities|Array|MakeArray (.WorldTransform (Utilities|Struct|BreakBoneInfo _boneinfo_4)) (.WorldTransform (Utilities|Struct|BreakBoneInfo _boneinfo_5)) (.WorldTransform (Utilities|Struct|BreakBoneInfo _boneinfo_6)) (Math|Transform|MakeTransform (+ (.location (.WorldTransform (Utilities|Struct|BreakBoneInfo _boneinfo_6))) "0.000000,0.000000,-15.000000") (.rotation (.WorldTransform (Utilities|Struct|BreakBoneInfo _boneinfo_6))) (.scale (.WorldTransform (Utilities|Struct|BreakBoneInfo _boneinfo_6))))) :PolylineTexParamU _array :bLoop false :StartScale 1.0 :EndScale 1.0 :RotationAngleDeg 0.0 :MiterLimit 1.0))
    (:then_2
      (bind (_bisvalidbonename_7 _boneinfo_7) (GeometryScript|MeshQueries|BoneWeights|GetBoneInfo _skelmeshdynamic_4 "thigh_l"))
      (bind (_bisvalidbonename_8 _boneinfo_8) (GeometryScript|MeshQueries|BoneWeights|GetBoneInfo _skelmeshdynamic_4 "calf_l"))
      (bind (_bisvalidbonename_9 _boneinfo_9) (GeometryScript|MeshQueries|BoneWeights|GetBoneInfo _skelmeshdynamic_4 "foot_l"))
      (GeometryScript|Primitives|AppendSweepPolyline :TargetMesh (Variables|GetTargetMesh) :PrimitiveOptions (Utilities|Struct|MakeGeometryScriptPrimitiveOptions "PerFace") :PolylineVertices _array_2 :SweepPath (Utilities|Array|MakeArray (.WorldTransform (Utilities|Struct|BreakBoneInfo _boneinfo_7)) (.WorldTransform (Utilities|Struct|BreakBoneInfo _boneinfo_8)) (Math|Transform|MakeTransform (+ (.location (.WorldTransform (Utilities|Struct|BreakBoneInfo _boneinfo_9))) "0.000000,0.000000,-15.000000") (.rotation (.WorldTransform (Utilities|Struct|BreakBoneInfo _boneinfo_9))) (.scale (.WorldTransform (Utilities|Struct|BreakBoneInfo _boneinfo_9))))) :PolylineTexParamU _array :bLoop false :StartScale 1.0 :EndScale 1.0 :RotationAngleDeg 0.0 :MiterLimit 1.0))
    (:then_3
      (bind (_bisvalidbonename_10 _boneinfo_10) (GeometryScript|MeshQueries|BoneWeights|GetBoneInfo _skelmeshdynamic_3 "thigh_r"))
      (bind (_bisvalidbonename_11 _boneinfo_11) (GeometryScript|MeshQueries|BoneWeights|GetBoneInfo _skelmeshdynamic_3 "calf_r"))
      (bind (_bisvalidbonename_12 _boneinfo_12) (GeometryScript|MeshQueries|BoneWeights|GetBoneInfo _skelmeshdynamic_3 "foot_r"))
      (GeometryScript|Primitives|AppendSweepPolyline :TargetMesh (Variables|GetTargetMesh) :PrimitiveOptions (Utilities|Struct|MakeGeometryScriptPrimitiveOptions "PerFace") :PolylineVertices _array_2 :SweepPath (Utilities|Array|MakeArray (.WorldTransform (Utilities|Struct|BreakBoneInfo _boneinfo_10)) (.WorldTransform (Utilities|Struct|BreakBoneInfo _boneinfo_11)) (Math|Transform|MakeTransform (+ (.location (.WorldTransform (Utilities|Struct|BreakBoneInfo _boneinfo_12))) "0.000000,0.000000,-15.000000") (.rotation (.WorldTransform (Utilities|Struct|BreakBoneInfo _boneinfo_12))) (.scale (.WorldTransform (Utilities|Struct|BreakBoneInfo _boneinfo_12))))) :PolylineTexParamU _array :bLoop false :StartScale 1.0 :EndScale 1.0 :RotationAngleDeg 0.0 :MiterLimit 1.0))
    (:then_4
      (bind (_bisvalidbonename_13 _boneinfo_13) (GeometryScript|MeshQueries|BoneWeights|GetBoneInfo _skelmeshdynamic "head"))
      (bind (_bisvalidbonename_14 _boneinfo_14) (GeometryScript|MeshQueries|BoneWeights|GetBoneInfo _skelmeshdynamic "neck_01"))
      (bind (_bisvalidbonename_15 _boneinfo_15) (GeometryScript|MeshQueries|BoneWeights|GetBoneInfo _skelmeshdynamic "spine_03"))
      (bind (_bisvalidbonename_16 _boneinfo_16) (GeometryScript|MeshQueries|BoneWeights|GetBoneInfo _skelmeshdynamic "pelvis"))
      (GeometryScript|Primitives|AppendSweepPolyline :TargetMesh (Variables|GetTargetMesh) :PrimitiveOptions (Utilities|Struct|MakeGeometryScriptPrimitiveOptions "PerFace") :PolylineVertices _array_2 :SweepPath (Utilities|Array|MakeArray (Math|Transform|MakeTransform (+ (.location (.WorldTransform (Utilities|Struct|BreakBoneInfo _boneinfo_13))) "0.000000,0.000000,15.000000") (.rotation (.WorldTransform (Utilities|Struct|BreakBoneInfo _boneinfo_13))) (.scale (.WorldTransform (Utilities|Struct|BreakBoneInfo _boneinfo_13)))) (.WorldTransform (Utilities|Struct|BreakBoneInfo _boneinfo_14)) (.WorldTransform (Utilities|Struct|BreakBoneInfo _boneinfo_15)) (Math|Transform|MakeTransform (+ (.location (.WorldTransform (Utilities|Struct|BreakBoneInfo _boneinfo_16))) "0.000000,0.000000,-15.000000") (.rotation (.WorldTransform (Utilities|Struct|BreakBoneInfo _boneinfo_16))) (.scale (.WorldTransform (Utilities|Struct|BreakBoneInfo _boneinfo_16))))) :PolylineTexParamU _array :bLoop false :StartScale 1.0 :EndScale 1.0 :RotationAngleDeg 0.0 :MiterLimit 1.0))
    (:then_5
      (bind _returnvalue_2 (GeometryScript|Normals|RecomputeNormals TargetMesh))
      (GeometryScript|Normals|ComputeTangents _returnvalue_2))))

(fn makeStaticMesh_skelMeshes
  (params (mesh :type Dynamic%20Mesh%20Object%20Reference) (lod :type Integer) (GenerateRibbonLOD :type Boolean))
  (bind _returnvalue (Utilities|String|ToString(Integer) lod))
  (bind _localdynamicmesh (LocalVariables|GetlocalDynamicMesh))
  (bind _returnvalue_2 (GeometryScript|MeshQueries|GetNumVertexIDs _localdynamicmesh))
  (bind _returnvalue_3 (Utilities|String|ToString(Integer) _returnvalue_2))
  (bind _localdynamicmesh_2 (LocalVariables|GetlocalDynamicMesh))
  (bind _returnvalue_4 (GeometryScript|MeshQueries|GetNumVertexIDs _localdynamicmesh_2))
  (bind _returnvalue_5 (Math|Conversions|ToFloat(Integer) _returnvalue_4))
  (bind _returnvalue_6 (Utilities|String|Append "DBG_LOD = " _returnvalue))
  (bind _returnvalue_7 (Utilities|String|Append _returnvalue_6 "   "))
  (bind _returnvalue_8 (/ _returnvalue_5 1e+02))
  (bind _prop_autogeneratelod_reduction (Variables|Prop|GetProp_AutoGenerateLOD_reduction))
  (bind _localdynamicmesh_3 (LocalVariables|GetlocalDynamicMesh))
  (bind _autogeneratelod_reduction (Variables|Default|GetAutoGenerateLOD_reduction))
  (bind _prop_skeletalmesh (Variables|Prop|GetProp_SkeletalMesh))
  (if GenerateRibbonLOD
    ;[1 1 1] Experimental see /Script/Engine.Blueprint'/Game/Developers/PaulMader/CharacterWPOBake/TestRibbon.TestRibbon'
    (bind _returnvalue_9 (DynamicMeshActor|AllocateComputeMesh))
    (bind _output_get (LocalVariables|SetMergedSkelMesh _returnvalue_9))
    (Utilities|FlowControl|Sequence
      (:then_1
        (GeometryScript|SkeletalMesh|CopyMeshfromSkeletalMesh "/Game/Characters/Player/Male/Medium/Base/SK_M_MALE_Base.SK_M_MALE_Base" _output_get (struct :bApplyBuildSettings true :bRequestTangents true :bIgnoreRemoveDegenerates false :bUseBuildScale true) (struct :LODType "MaxAvailable" :LODIndex 0)
          (:Success
            (bind _returnvalue_10 (DynamicMeshActor|AllocateComputeMesh))
            (bind _output_get_2 (LocalVariables|SetRibbonMesh _returnvalue_10))
            (Default|MakeRibbonMesh _output_get_2 (LocalVariables|GetMergedSkelMesh))
            (CallFunction|StealVertexColor (LocalVariables|GetRibbonMesh))
            (bind _returnvalue_11 (GeometryScript|MeshEdits|AppendMesh :TargetMesh mesh :AppendMesh (LocalVariables|GetRibbonMesh) :bDeferChangeNotifications false :AppendOptions (struct :CombineMode "EnableAllMatching")))
            (return _returnvalue_11)))))
    (else
      (for (_array_element _array_index) (Variables|Default|GetskeletalMeshes)
        (Components|SkinnedMesh|SetSkinnedAssetandUpdate (Variables|Default|GetlocalSkeletalMesh) _array_element true)
        (bind _returnvalue_12 (DynamicMeshActor|AllocateComputeMesh))
        (LocalVariables|SetlocalDynamicMesh _returnvalue_12)
        (GeometryScript|SkeletalMesh|CopyMeshfromSkeletalMesh _array_element (LocalVariables|GetlocalDynamicMesh) (struct :bApplyBuildSettings true :bRequestTangents true :bIgnoreRemoveDegenerates false :bUseBuildScale true) (Utilities|Struct|MakeGeometryScriptMeshReadLOD "MaxAvailable" (Math|Integer|SelectInt 0 lod (Variables|Default|GetAutoGenerateLODs)))
          (:Failure
            (Development|PrintString (Utilities|String|Append (Utilities|String|Append _returnvalue_7 (Utilities|GetDisplayName _array_element)) "  FAILED
") false false))
          (:Success
            (Development|PrintString (Utilities|String|Append (Utilities|String|Append (Utilities|String|Append _returnvalue_7 (Utilities|GetDisplayName _array_element)) "  Vertcount = ") _returnvalue_3) false)
            (CallFunction|DeleteMeshbyVertexColor (LocalVariables|GetlocalDynamicMesh) _array_element)
            (Utilities|FlowControl|Sequence
              (:then_0
                ;[1 1 1] AutoGenerate LODs (currently the CopyLOD from Skelmesh is not working with Autogenrated lods this is a work around and generates LOD. its alos a quick way to optmize for the bake
                (if (Variables|Default|GetAutoGenerateLODs)
                  (if (< (Utilities|Array|Get(acopy) (select (== _prop_skeletalmesh _array_element) _prop_autogeneratelod_reduction _autogeneratelod_reduction) lod) 1e+02)
                    (GeometryScript|Simplification|ApplySimplifytoVertexCount (LocalVariables|GetlocalDynamicMesh) (Math|Float|Round (* _returnvalue_8 (Utilities|Array|Get(acopy) (select (== _prop_skeletalmesh _array_element) _prop_autogeneratelod_reduction _autogeneratelod_reduction) lod))) (Utilities|Struct|MakeGeometryScriptSimplifyMeshOptions "AttributeAware"))
                    (Development|PrintString (Utilities|String|Append "Optimizing to VertCount " (Utilities|String|ToString(Integer) (Math|Float|Round (* _returnvalue_8 (Utilities|Array|Get(acopy) (select (== _prop_skeletalmesh _array_element) _prop_autogeneratelod_reduction _autogeneratelod_reduction) lod)))) "   /// Org" (Utilities|String|ToString(Integer) _returnvalue_4) "   /// Assetname " (Utilities|GetDisplayName _array_element)) false))))
              (:then_1
                (CallFunction|SetMaterialIDandStoreTexture (LocalVariables|GetlocalDynamicMesh) _array_element)
                (if (Variables|Default|GetfixPose)
                  (CallFunction|FixBindPoseRotation (LocalVariables|GetlocalDynamicMesh)))
                (if (Variables|Default|GetNaniteEnabled)
                  (CallFunction|SetAnimTextureUVs _localdynamicmesh_3)
                  (else
                    (CallFunction|SetVertexColor _localdynamicmesh_3)))
                (Development|PrintString (Utilities|String|Append (Utilities|String|Append (Utilities|String|Append _returnvalue_7 (Utilities|GetDisplayName _array_element)) "  Vertcount = ") _returnvalue_3) false)
                (GeometryScript|MeshEdits|AppendMesh :TargetMesh mesh :AppendMesh (LocalVariables|GetlocalDynamicMesh) :bDeferChangeNotifications false :AppendOptions (struct :CombineMode "EnableAllMatching")))))))
      (return mesh))))

(fn setMaterialIDandStoreTexture
  (params (TargetMesh :type Dynamic%20Mesh%20Object%20Reference) (SkelMesh :type Skeletal%20Mesh%20Object%20Reference))
  (bind _materials (Variables|/Script/Engine.SkeletalMesh|GetMaterials SkelMesh))
  (bind _returnvalue (Utilities|Array|LastIndex _materials))
  (bind _nummaterialids (Variables|Default|GetnumMaterialIds))
  (for _index (range (+ _returnvalue 1))
    (GeometryScript|Materials|RemapMaterialIDs TargetMesh (- _returnvalue _index) (+ (- _returnvalue _index) _nummaterialids))
    (Development|PrintString (Utilities|String|Append "DBGMAT = numats = " (Utilities|String|ToString(Integer) _nummaterialids) "  // remapping  " (Utilities|String|ToString(Integer) (- _returnvalue _index)) " /// to " (Utilities|String|ToString(Integer) (+ (- _returnvalue _index) _nummaterialids)) "/// asset" (Utilities|GetDisplayName SkelMesh)))
    (Utilities|Array|Add (Variables|Default|Getmaterials) "/Game/ArtTools/characterWPO/MI_StaticMeshMaker.MI_StaticMeshMaker")
    (Utilities|Array|Add (Variables|Default|GetmaterialNames) (Utilities|String|StringToName ""))
    (bind _returnvalue_2 (Rendering|Material|CreateDynamicMaterialInstance (Utilities|Array|Get(acopy) (Rendering|Material|GetMaterials (Variables|Default|GetlocalSkeletalMesh)) _index) "None" "None" :!class /Script/Engine.KismetMaterialLibrary))
    (bind _returnvalue_3 (Rendering|Material|GetTextureParameterValue _returnvalue_2 "Diffuse" :!class /Script/Engine.MaterialInstanceDynamic))
    (Utilities|Array|Add (Variables|Default|GetdiffuseTextures) _returnvalue_3)
    (bind _returnvalue_4 (Rendering|Material|GetTextureParameterValue _returnvalue_2 "Normals" :!class /Script/Engine.MaterialInstanceDynamic))
    (Utilities|Array|Add (Variables|Default|GetnormalTextures) _returnvalue_4))
  (Variables|Default|SetnumMaterialIds (+ (Variables|Default|GetnumMaterialIds) (Utilities|Array|Length _materials))))

(fn StealVertexColor
  (params (TargetMesh :type Dynamic%20Mesh%20Object%20Reference))
  (bind _returnvalue (DynamicMeshActor|AllocateComputeMesh))
  (bind _returnvalue_2 (GeometryScript|StaticMesh|CopyMeshFromStaticMeshwithSectionMaterials :FromStaticMeshAsset (Variables|Default|GettargetStaticMesh) :ToDynamicMesh _returnvalue :RequestedLOD (Utilities|Struct|MakeGeometryScriptMeshReadLOD "MaxAvailable"))
    (:Success
      (bind _outputbvh (GeometryScript|Spatial|BuildBVHForMesh _returnvalue_2))
      (for _index (range (GeometryScript|MeshQueries|GetNumVertexIDs TargetMesh))
        (bind (_nearestresult _returnvalue_3) (GeometryScript|Spatial|FindNearestPointonMesh _returnvalue_2 _outputbvh (.ReturnValue (GeometryScript|MeshQueries|GetVertexPosition TargetMesh _index)))
          (:Found
            (bind _color1 (GeometryScript|MeshQueries|GetTriangleVertexColors _returnvalue_3 (.TriangleID (Utilities|Struct|BreakGeometryScriptTrianglePoint _nearestresult))))
            (Utilities|Array|Add (LocalVariables|GetVertexColors) _color1))))
      (bind _colorlist (GeometryScript|ListUtils|ConvertArraytoColorList (LocalVariables|GetVertexColors)))
      (GeometryScript|VertexColor|SetMeshPerVertexColors TargetMesh _colorlist))))

(fn ConstructionScript
  (bind _previewmannequin (Variables|Default|GetpreviewMannequin))
  (Utilities|FlowControl|Sequence
    (:then_0
      (switch Utilities|FlowControl|Switch|SwitchonCrowd_Baker_Presets (Variables|Default|GetBoneQualityPreset)
        (:NewEnumerator0
          (Variables|Default|SetboneNames (Variables|/Game/ArtTools/characterWPO/BonePreset.BonePreset_C|GetBones (Variables|Default|GetMediumPreset))))
        (:NewEnumerator1
          (Variables|Default|SetboneNames (Variables|/Game/ArtTools/characterWPO/BonePreset.BonePreset_C|GetBones (Variables|Default|GetHighPreset))))))
    (:then_1
      ;[1 1 1] Fix for Copy Paste Bone List
      (for (_array_element _array_index) (Variables|Default|GetboneNames)
        (Utilities|Array|SetArrayElem (Variables|Default|GetboneNames) _array_index (Utilities|String|StringToName (Utilities|String|Replace (Utilities|String|ToString(Name) _array_element) "\"")))))
    (:then_2
      ;[1 1 1] Fix for Copy Paste Bone List
      (for (_array_element_2 _array_index_2) (Variables|Prop|GetProp_Bones)
        (Utilities|Array|SetArrayElem (Variables|Prop|GetProp_Bones) _array_index_2 (Utilities|String|StringToName (Utilities|String|Replace (Utilities|String|ToString(Name) _array_element_2) "\"")))))
    (:then_3
      (CallFunction|ClearArrays)
      ;[1 1 1] Preview Mannequin
      (Utilities|IsValid _previewmannequin
        (:Is%20Valid
          (Customization|SetAthenaCharacterDefandVariants _previewmannequin (Variables|Default|GetCID) (Utilities|Array|MakeArray))
          ;[1 1 1] skelmeshes
          (for _array_element_3 (Variables|/Script/FortniteGame.AthenaCharacterItemDefinition|GetBaseCharacterParts (Variables|Default|GetCID))
            (bind _returnvalue (Utilities|LoadAssetBlocking _array_element_3))
            (bind _ascustom_character_part (Utilities|Casting|CastToCustomCharacterPart _returnvalue))
            (Utilities|Array|Add (Variables|Default|GetskeletalMeshes) (CharacterPart|Art|GetSkeletalMesh _ascustom_character_part))))))
    (:then_4
      (if (Variables|Prop|GetOutputPropOnly)
        (Utilities|Array|Clear (Variables|Default|GetskeletalMeshes))))))
