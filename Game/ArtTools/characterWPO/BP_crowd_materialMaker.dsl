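; BP_crowd_materialMaker is an art-tool Blueprint that launches a Python material script for the crowd character WPO (World Position Offset) baking workflow.
; Its materials function appends its own object path and a "Run" argument to a "py" command for Techart\WPOBaker\materials.py, prints it, then executes it as a console command.
; The material work itself lives in that external script; the event graph and construction script are empty.
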(event EventBeginPlay)

(event Collision|EventActorBeginOverlap
  (params (OtherActor :type Actor%20Object%20Reference)))

(event EventTick
  (params (DeltaSeconds :type Float%20%28single-precision%29)))

(fn materials
  (bind _self self)
  (bind _returnvalue (Utilities|GetObjectPathString _self))
  (bind _returnvalue_2 (Utilities|String|Append "py Techart\\WPOBaker\\materials.py " _returnvalue " Run"))
  (Development|PrintString _returnvalue_2)
  (Development|ExecuteConsoleCommand _returnvalue_2 (Game|GetPlayerController 0)))

(fn ConstructionScript)
