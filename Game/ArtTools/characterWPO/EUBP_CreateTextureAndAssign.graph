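# Editor utility that bakes a render target into a texture for a static mesh and assigns it to that mesh's materials, running from the Construction Script.
# It names the texture from the mesh path plus "_RT", overwriting an existing asset or creating an HDR, mipless, nearest-filtered one.
# It then sets that texture as "AnimationTexture" on each of the mesh's Material Instance Constants.
# The EventGraph's CreateAndAssign event only prints the mesh path.
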EditorUtilityBlueprint /Game/ArtTools/characterWPO/EUBP_CreateTextureAndAssign
graphs [
 Create and Assign Texture {nodes [
   Create and Assign Texture() {is=title K2Node_FunctionEntry}
   > n1 = Sequence()
   n2 = Split(SourceString=Get Object Path String(Object=Get Staticmesh() -> (Staticmesh:Static Mesh Object Reference)), InStr=., SearchDir=FromStart)
   > PrintString(InString=@n2)
   n3 = Append(A=@n2, B=_RT)
   > n4 = DoesAssetExist(AssetPath=@n3)
   > n5 = Branch(Condition=@n4)
   > n6 = LoadAsset(AssetPath=@n3)
   > n7 = Cast To Texture2D(Object=@n6)
   n8 = Get Render Target() -> (Render Target:Texture Render Target 2D Object Reference)
   > ConvertRenderTargetToTexture2DEditorOnly(RenderTarget=@n8, Texture=@n7)
   n9 = Render Target 2D Create Static Texture 2D Editor Only!(
    execute=@n5.else
    RenderTarget:Texture Render Target 2D Object Reference=@n8
    Name:String=@n3
    CompressionSettings:TextureCompressionSettings Enum=TC_HDR
    MipSettings:TextureMipGenSettings Enum=TMGS_NoMipmaps
   ) -> (then, ReturnValue:Texture 2D Object Reference)
   > Set Filter!(execute, Filter:TextureFilter Enum=TF_Nearest, self:Texture Object Reference=@n9) -> (then, Output_Get:TextureFilter Enum=TF_Nearest)
   n10 = LoadAsset(execute=@n1.then_1, AssetPath=@n3)
   > n11 = Cast To Texture2D(Object=@n10)
   > n12 = For Each Loop(Array:Array of Static Material Structures=Get StaticMaterials!(self:Static Mesh Object Reference=Get Staticmesh()) -> (StaticMaterials:Array of Static Material Structures)) -> (Array Element:Static Material Structure)
   n13 = Break Static Material!(StaticMaterial:"Static Material Structure (by ref)"=@n12) -> (
    MaterialInterface:Material Interface Object Reference
    MaterialSlotName:Name
    UVChannelData:Mesh UVChannel Info Structure
    OverlayMaterialInterface:Material Interface Object Reference
   ) {is=title K2Node_BreakStruct}
   > n14 = Cast To MaterialInstanceConstant(Object=@n13)
   > Set MIC Texture Parameter Editor Only!(
    execute
    Material:Material Instance Constant Object Reference=@n14
    ParamName:String=AnimationTexture
    Texture:Texture 2D Object Reference=@n11
   ) -> (then, ReturnValue:bool) {is=title K2Node_CallFunction}
   GetPathNameForLoadedAsset(LoadedAsset=@n13)
  ]}
 EventGraph {nodes [
   CreateAndAssign() {is=title K2Node_CustomEvent}
   n1 = Get Staticmesh()
   > PrintString(InString=Split(SourceString=Get Object Path String(Object=@n1), InStr=., SearchDir=FromStart))
   Render Target 2D Create Static Texture 2D Editor Only(RenderTarget=Get Render Target(), Name=Texture, CompressionSettings=TC_Default, MipSettings=TMGS_FromTextureGroup)
   ConvertRenderTargetToTexture2DEditorOnly()
   Event BeginPlay()
   Event ActorBeginOverlap()
   Event Tick!() -> (OutputDelegate, then, DeltaSeconds:float)
  ]}
 UserConstructionScript {nodes [
   Construction Script()
   > Create and Assign Texture() {is=title K2Node_CallFunction}
  ]}
]
variables [
 Render Target {type=TextureRenderTarget2D, default=null}
 Staticmesh {type=StaticMesh, default=null}
]
