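# Material function that animates characters by World Position Offset from baked transforms in a sub-region of AnimationTexture.
# Vertex color R, or UV1.g for Nanite, picks a row within stacked position, quaternion and optional scale bands at the given u.
# Pre-skinned vertices are scaled and rotated around their sampled bind-pose pivot, then translated.
# Outputs switchable WPO, a normal rotated by the interpolated quaternion, and the quaternion itself.
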MaterialFunction /Game/ArtTools/characterWPO/MF_WPOTransform_subuvs
graph {nodes [
  n1 = StaticSwitchParameter(True=ComponentMask(None=TextureCoordinate() {coordinateIndex=1}) {g=true}, False=VertexColor().R) {parameterName=UseWithNaniteEnabled}
  n2 = Multiply(
   A=FunctionInput() {inputName=vHeight, sortPriority=11}
   B=StaticSwitchParameter(True=Constant() {r=0.3333}, False=Constant() {r=0.5}) {parameterName=ConstainsScaleInfo}
  )
  n3 = Add(
   A=Add(A=Multiply(A=@n1, B=@n2) {constB=0.5}, B=FunctionInput() {inputName=vStart, sortPriority=10})
   B=Multiply(A=FunctionInput() {desc="height of a texel (1/yRez)", inputName=texelHeight, inputType=FunctionInput_Scalar, sortPriority=12}) {constB=0.5}
  ) {constB=-0.1}
  n4 = FunctionInput() {desc=u value to sample, inputName=u, inputType=FunctionInput_Scalar}
  n5 = TextureSampleParameter2D(UVs=MaterialFunctionCall(X=@n4, Y=@n3) {materialFunction=&MakeFloat2}.Result) {
   desc=position
   group=BakedTransforms
   parameterName=AnimationTexture
   samplerType=SAMPLERTYPE_LinearColor
   sortPriority=0
   texture=/Game/Global/Textures/Default/Blanks/T_BlackVectorDisplacement
  }
  n6 = Add(A=@n3, B=@n2) {constB=0.5}
  n7 = TextureSampleParameter2D(UVs=MaterialFunctionCall(X=@n4, Y=Add(A=@n6, B=@n2)) {materialFunction=&MakeFloat2}.Result) {
   desc=position
   group=BakedTransforms
   parameterName=AnimationTexture
   samplerType=SAMPLERTYPE_LinearColor
   sortPriority=0
   texture=&T_BlackVectorDisplacement
  }
  n8 = PreSkinnedPosition()
  n9 = TextureSampleParameter2D(UVs=MaterialFunctionCall(X=FunctionInput() {desc=u value of bind pose, inputName=bind_u, inputType=FunctionInput_Scalar, sortPriority=1}, Y=@n3) {materialFunction=&MakeFloat2}.Result) {
   desc=position
   group=BakedTransforms
   parameterName=AnimationTexture
   samplerType=SAMPLERTYPE_LinearColor
   sortPriority=0
   texture=&T_BlackVectorDisplacement
  }
  n10 = Subtract(A=@n8, B=@n9.RGB) {constB=0}
  n11 = Multiply(A=@n10, B=@n7.RGB) {constB=0.5}
  n12 = TextureSampleParameter2D(UVs=MaterialFunctionCall(X=@n4, Y=@n6) {materialFunction=&MakeFloat2}.Result) {
   desc=position
   group=BakedTransforms
   parameterName=AnimationTexture
   samplerType=SAMPLERTYPE_LinearColor
   sortPriority=0
   texture=&T_BlackVectorDisplacement
  }
  n13 = MaterialFunctionCall!(
   Vector=StaticSwitchParameter(True=StaticSwitchParameter(True=@n11, False=@n10) {group=Deformation, parameterName=ApplyScaleAnimation, sortPriority=10}, False=@n10) {group=BakedTransforms, parameterName=ConstainsScaleInfo, sortPriority=25}
   Quat=@n12.RGBA
   Invert Quaternion
  ) {materialFunction=/Engine/Functions/Engine_MaterialFunctions03/Math/MultiplyVectorWithQuaternion}
  n14 = Subtract(A=Add(A=Add(A=@n5.RGB, B=@n13.Result) {constB=0}, B=@n9.RGB) {constB=0}, B=@n8)
  FunctionOutput(None=StaticSwitchParameter(True=Transform(None=@n14) {transformSourceType=TRANSFORMSOURCE_Local}, False=Constant3Vector()) {defaultValue=true, group=Deformation, parameterName=ApplyWPO, sortPriority=5}) {outputName=WPO}
  n15 = Transform(None=FunctionInput(Preview=VertexNormalWS()) {bUsePreviewValueAsDefault=true, inputName=WorldNormal, sortPriority=50}) {transformSourceType=TRANSFORMSOURCE_World, transformType=TRANSFORM_Instance}
  n16 = VertexInterpolator(VS=@n12.RGBA)
  FunctionOutput(None=Transform(None=MaterialFunctionCall(Vector=@n15, Quat=@n16.PS) {materialFunction=&MultiplyVectorWithQuaternion}.Result) {transformSourceType=TRANSFORMSOURCE_Instance}) {outputName=Normal, sortPriority=-2}
  FunctionOutput(None=@n16.PS) {outputName=Quaternion}
  FunctionOutput(None=@n12.RGBA) {outputName=QuaternionVS}
  LinearInterpolate(A=Constant4Vector() {constant=<rgba 0 0 0 1>}.RGBA, B=@n12.RGBA, Alpha=ScalarParameter() {defaultValue=1, parameterName=QuaternionMix})
 ]}
