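# Master material for characters animated by textures through world position offset, usable on instanced static meshes and Niagara mesh particles.
# With TextureIsAnimationAtlas on, it reads each clip's sub-UV rectangle from atlas header texels, selected by animIndex or a particle dynamic parameter.
# Time, offset and per-instance random drive looping playback, which dynamic parameters can override in atlas mode.
# MF_WPOTransform_subuvs samples the animation, and its normal blends with a normal map.
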Material /Game/ArtTools/characterWPO/M_characterWPOAtlas_Master
graph {
 properties {
  bTangentSpaceNormal=false
  bUsedWithInstancedStaticMeshes=true
  bUsedWithNiagaraMeshParticles=true
  pixelDepthOffsetMode=PDOM_Legacy
  usedShadingModels=Default Lit
 }
 nodes [
  Constant() {r=4}
  Constant() {r=7.222}
  n1 = DynamicParameter() {defaultValue=<rgba 0 0 0 0>}
  MaterialFunctionCall(X=@n1.Param1, Y=@n1.Param1, Z=@n1.Param1, A=@n1.Param1) {materialFunction=&MakeFloat4}
  n2 = NamedRerouteUsage()
  n3 = Add(A=@n2) {constB=0.0001}
  n4 = Time()
  n5 = ScalarParameter() {parameterName=TimeOffset}
  n6 = Multiply(A=Add(A=Add(A=@n4, B=@n5), B=PerInstanceRandom()), B=Divide(B=NamedRerouteUsage()) {constA=30})
  n7 = Frac(None=Multiply(A=Divide(B=@n3), B=@n6))
  n8 = NamedRerouteUsage()
  n9 = Add(A=@n8) {constB=0.0001}
  n10 = DynamicParameter() {defaultValue=<rgba 0 0 0 0>}
  n11 = Divide(B=NamedRerouteUsage()) {constA=30}
  n12 = Frac(None=Multiply(A=Divide(B=@n9), B=Multiply(A=@n10.Param2, B=@n11)))
  n13 = Add(A=LinearInterpolate(A=Multiply(A=@n2, B=@n7), B=Multiply(A=@n8, B=@n12), Alpha=@n10.Param1), B=NamedRerouteUsage())
  n14 = NamedRerouteUsage()
  n15 = Add(A=Multiply(A=Frac(None=@n6), B=Subtract(B=@n14)), B=@n14)
  n16 = NamedRerouteUsage()
  n17 = StaticSwitchParameter(True=Multiply(A=@n16) {constB=1.5}, False=Multiply(A=@n16) {constB=0.5}) {parameterName=TextureIsAnimationAtlas}
  n18 = NamedRerouteUsage()
  n19 = Constant()
  n20 = NamedRerouteUsage()
  n21 = Constant() {r=1}
  MaterialFunctionCall(
   BaseNormal=MaterialFunctionCall!(
    u=StaticSwitchParameter(True=@n13, False=@n15) {parameterName=TextureIsAnimationAtlas}
    bind_u=Add(A=@n17)
    vStart=StaticSwitchParameter(True=@n18, False=@n19) {parameterName=TextureIsAnimationAtlas}
    vHeight=StaticSwitchParameter(True=@n20, False=@n21) {parameterName=TextureIsAnimationAtlas}
    texelHeight=NamedRerouteUsage()
    WorldNormal
   ) {materialFunction=&/MF_WPOTransform_subuvs}.Normal
   AdditionalNormal=TextureSampleParameter2D() {parameterName=Normals, samplerType=SAMPLERTYPE_Normal, texture=&FlatNormalBlue}.RGB
  ) {materialFunction=&BlendAngleCorrectedNormals}
  n22 = TextureObjectParameter() {parameterName=AnimationTexture, texture=&T_BlacksRGB}
  n23 = MaterialFunctionCall(Float2=TextureProperty(Tex=@n22)) {materialFunction=&BreakOutFloat2Components}
  n24 = Divide(B=NamedRerouteDeclaration(None=@n23.R) {name=textureWidth, nodeColor=<rgba 0.0078 0 1 1>})
  NamedRerouteDeclaration(None=Multiply(A=NamedRerouteDeclaration(None=@n24) {name=texelWidth, nodeColor=<rgba 0 1 0.2266 1>}) {constB=0.5}) {name=halfTexelWidth, nodeColor=<rgba 0.0312 0 1 1>}
  NamedRerouteDeclaration(None=Divide(B=@n23.G)) {name=texelHeight, nodeColor=<rgba 0 1 0.2266 1>}
  n25 = NamedRerouteUsage()
  n26 = DynamicParameter() {defaultValue=<rgba 0 0 0 0>}
  n27 = Multiply(A=LinearInterpolate(A=ScalarParameter() {parameterName=animIndex}, B=@n26.Param3, Alpha=@n26.Param3), B=@n25)
  n28 = Multiply(A=@n25) {constB=0.5}
  n29 = TextureSample(
   UVs=MaterialFunctionCall(X=NamedRerouteUsage(), Y=Add(A=@n27, B=@n28)) {materialFunction=&MakeFloat2}.Result
   Tex=NamedRerouteUsage()
  ) {texture=/Engine/EngineResources/DefaultTexture}
  n30 = NamedRerouteUsage()
  NamedRerouteDeclaration(None=Subtract(A=@n29.B, B=@n30)) {name=uWidth, nodeColor=<rgba 0 1 0.6016 1>}
  NamedRerouteDeclaration(None=@n29.G) {name=vStart, nodeColor=<rgba 0.6875 0 1 1>}
  NamedRerouteDeclaration(None=@n29.A) {name=vHeight, nodeColor=<rgba 0 1 0.2266 1>}
  NamedRerouteDeclaration(None=@n22) {name=AnimationTexture, nodeColor=<rgba 0.6172 0 1 1>}
  NamedRerouteDeclaration(None=Add(A=@n30, B=@n29.R)) {name=uStart, nodeColor=<rgba 1 0 0.0469 1>}
  TextureSampleParameter2D() {parameterName=BaseColor, texture=&T_BlacksRGB}
 ]
}
